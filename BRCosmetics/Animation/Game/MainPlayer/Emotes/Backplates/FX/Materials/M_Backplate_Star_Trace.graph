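# Unlit translucent Niagara sprite and mesh particle material for a star trace effect in emote backplate FX.
# Particle color times a dynamic EmissiveMult feeds MF_FX_EmissiveAdjustments for glow.
# A UV-based wipe mask, scaled, animated and offset by dynamic parameters, is saturated and multiplied by particle alpha.
# Static switches choose a vertical linear gradient with optional inversion or a horizontal sine or parabolic profile.
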Material /BRCosmetics/Animation/Game/MainPlayer/Emotes/Backplates/FX/Materials/M_Backplate_Star_Trace
graph {
 properties {
  bAutomaticallySetUsageInEditor=false
  bUsedWithNiagaraMeshParticles=true
  bUsedWithNiagaraSprites=true
  blendMode=BLEND_Translucent
  shadingModel=MSM_Unlit
  usedShadingModels=Unlit
 }
 nodes [
  n1 = ParticleColor()
  MaterialFunctionCall(Emissive=Multiply(A=@n1.RGB, B=DynamicParameter() {parameterIndex=1, paramNames [EmissiveMult,Param2,Param3,Param4]}.EmissiveMult)) {materialFunction=&MF_FX_EmissiveAdjustments}
  n2 = DynamicParameter() {defaultValue=<rgba 1 0 0.67 1>, paramNames [WipeScale,Anim,Offset,WipeMult]}
  n3 = ComponentMask(None=Multiply(A=TextureCoordinate(), B=@n2.WipeScale)) {g=true}
  n4 = StaticSwitchParameter(True=@n3, False=OneMinus(None=@n3)) {parameterName=Reverse Vert?}
  n5 = Multiply(A=TextureCoordinate(), B=@n2.WipeScale)
  n6 = ComponentMask(None=@n5) {r=true}
  n7 = OneMinus(None=@n5)
  n8 = StaticSwitchParameter(True=Abs(None=Sine(None=@n6)), False=Multiply(A=ComponentMask(None=@n7) {r=true}, B=ComponentMask(None=@n5) {r=true})) {parameterName=Reverse Horiz?}
  n9 = Add(
   A=Multiply(A=StaticSwitchParameter(True=@n4, False=@n8) {defaultValue=true, parameterName="Vert (T). Horiz (F)."}, B=@n2.WipeMult) {constB=0.2}
   B=@n2.Anim
  )
  Multiply(A=@n1.A, B=Saturate(None=Subtract(A=@n9, B=@n2.Offset)))
 ]
}
