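# Unlit, two-sided translucent material for tapered light rays in backplate emote FX, tinted by particle color.
# UV math shapes a soft-edged beam whose base thickness blends between side and above values by normal alignment with the instance Z axis.
# A reflected gradient fades the base, with depth fade skipped on Low and Medium quality.
# An exponential radial glow boosts emissive and opacity, and opacity fades at edge-on view angles.
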Material /BRCosmetics/Animation/Game/MainPlayer/Emotes/Backplates/FX/Materials/M_Rainbow_Rays
graph {
 properties {
  bUsedWithNiagaraSprites=true
  blendMode=BLEND_Translucent
  pixelDepthOffsetMode=PDOM_Legacy
  shadingModel=MSM_Unlit
  twoSided=true
  usedShadingModels=Unlit
 }
 nodes [
  n1 = TextureCoordinate()
  n2 = ComponentMask(None=@n1) {g=true}
  n3 = OneMinus(None=@n2)
  n4 = ScalarParameter() {defaultValue=0.2, parameterName=Base Thickness from Side}
  n5 = ScalarParameter() {defaultValue=0.4, parameterName=Base Thickness from Above}
  n6 = VertexNormalWS()
  n7 = Transform(None=Constant3Vector() {constant=<rgba 0 0 1 0>}) {transformSourceType=TRANSFORMSOURCE_Instance}
  n8 = Power(Base=Abs(None=DotProduct(A=@n6, B=@n7)), Exp=ScalarParameter() {defaultValue=2, parameterName=Base Thickness Lerp Exponent})
  n9 = OneMinus(None=Saturate(None=LinearInterpolate(A=@n4, B=@n5, Alpha=@n8)))
  n10 = Divide(A=ConstantBiasScale(None=ComponentMask(None=@n1) {r=true}) {bias=-0.5, scale=2}, B=OneMinus(None=Multiply(A=@n3, B=@n9)))
  n11 = DotProduct(A=@n10, B=@n10)
  n12 = SquareRoot(None=OneMinus(None=Saturate(None=@n11)))
  n13 = MaterialFunctionCall!(Gradient=@n2, Center Point=ScalarParameter() {defaultValue=0.05, parameterName=Base Fade}) {materialFunction=/Game/Effects/Fort_Effects/MaterialFunctions/MF_GradientReflected}
  n14 = DepthFade(Opacity=@n13.Result, FadeDistance=ScalarParameter() {defaultValue=8, parameterName=Depth Fade Distance})
  n15 = QualitySwitch(Default=@n14, Low=@n13.Result, High=@n14, Medium=@n13.Result, Epic=@n14)
  n16 = Multiply(A=Power(Base=Saturate(None=@n12), Exp=ScalarParameter() {defaultValue=2, parameterName=Edge Falloff}), B=Multiply(A=@n15, B=@n15))
  n17 = Multiply(
   A=MaterialFunctionCall(
    CenterPosition=Constant2Vector() {r=0.5}
    Radius=ScalarParameter() {defaultValue=0.5, parameterName=Central Glow Radius}
    Density=ScalarParameter() {defaultValue=1, parameterName=Central Glow Density}
   ) {materialFunction=&RadialGradientExponential}.RadialGradientExponential
   B=@n16
  )
  n18 = Multiply(A=@n17, B=ScalarParameter() {parameterName=Central Glow Intensity})
  n19 = ParticleColor()
  MaterialFunctionCall(Emissive=Add(A=Multiply(A=@n18, B=@n19.RGB), B=@n19.RGB)) {materialFunction=&MF_FX_EmissiveAdjustments}
  n20 = DotProduct(A=VertexNormalWS(), B=CameraVectorWS())
  Multiply(A=@n19.A, B=Multiply(A=Add(A=@n17, B=@n16), B=Abs(None=@n20)))
 ]
}
