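# Niagara system for a heart-themed emote backplate effect: a pink burst with rays, sparks and hearts.
# SimpleSpriteBurst pops two spawn-group meshes with a wobbly scale curve, material parameter animation, a glow sprite and a brief light.
# Sparks_and_Hearts splits two particle types by spawn group, scattering them with point, curl noise, drag and upward gravity forces, plus flickering glows.
# Rays adds spinning, fading sprites tinted from a system color array.
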NiagaraSystem /BRCosmetics/Animation/Game/MainPlayer/Emotes/Backplates/FX/NS_Backplates_Heart
dependencies {
 usedDataInterfaces [
  &NiagaraDataInterfaceSpriteRendererInfo
  &NiagaraDataInterfaceCurve
  &NiagaraDataInterfaceVector2DCurve
  &NiagaraDataInterfaceVectorCurve
  &NiagaraDataInterfaceArrayColor
 ]
 usedDynamicInputs [&MakeFloatFromVector2d]
 usedModules [
  &self:SystemSpawnScript.SystemScriptSource.SystemScriptGraph.NiagaraNodeAssignment_0.SetVariables_#1
  &SystemState
 ]
}
emitters {
 Rays {
  emitterSpawnScript [/SetVariables(Emitter.EmitterRandom:NiagaraInt32=&RandomRangeInt(Minimum=1, Maximum=500)) {isSetParameters=true}]
  emitterUpdateScript [
   &EmitterState()
   &SpawnBurst_Instantaneous(Spawn Count=10, Use Spawn Probability=false, UseLoopCountLimit=true)
   /SetVariables(Emitter.EdgeFade:NiagaraFloat=&FloatFromCurve(FloatCurve=DataInterface() {propertyValues {Curve {keys [<at 0 3>,<at 0.5 1>]}}}, CurveIndex=$Emitter.NormalizedLoopAge)) {isSetParameters=true}
  ]
  particleSpawnScript [&InitializeParticle(
    Color=&SelectLinearColorFromArray(Array Sampling Mode=::NewEnumerator0, Direct Array Index=?, Color Selection Array=$System.ColorPattern, Random Seed=0)
    Lifetime Max=1
    Lifetime Min=0.6
    Lifetime Mode=::NewEnumerator1
    Sprite Rotation Mode=::NewEnumerator2
    Sprite Size Max=<100 100>
    Sprite Size Min=<50 50>
    Sprite Size Mode=::NewEnumerator2
    UsePositionOffset=false
    Write Material Random=true
   )]
  particleUpdateScript [
   &ParticleState()
   &ScaleSpriteSize(
    Non-Uniform Curve Sprite Scale=DataInterface() {propertyValues {
      XCurve {keys [{tangentMode=RCTM_User, time=0, value=1}, {tangentMode=RCTM_User, time=1, value=0.5}]}
      YCurve {keys [
        {arriveTangent=15.8763, interpMode=RCIM_Cubic, leaveTangent=15.8763, tangentMode=RCTM_User, time=0, value=0}
        {
         arriveTangent=-0.3643
         arriveTangentWeight=0.1414
         interpMode=RCIM_Cubic
         leaveTangent=-0.3643
         leaveTangentWeight=0.33
         tangentMode=RCTM_User
         time=0.4
         value=1
        }
        {tangentMode=RCTM_User, time=1, value=0.9}
       ]}
     }}
    Scale Sprite Size Mode=::NewEnumerator3
   )
   &SpriteRotationRate(Rotation Rate=&RandomRangeFloat(Minimum=30, Maximum=75))
   &ScaleColor(
    Scale Alpha=&FloatFromCurve(
     FloatCurve=DataInterface() {propertyValues {Curve {keys [{arriveTangent=-2.6228, interpMode=RCIM_Cubic, leaveTangent=-2.6228, tangentMode=RCTM_User, time=0, value=1}, <at 1 0>]}}}
     Scale Curve=&RandomRangeFloat(Minimum=0.01, Maximum=0.03)
    )
    Scale RGB=&VectorFromFloat(Value=&FloatFromCurve(FloatCurve=DataInterface() {propertyValues {Curve {keys [{arriveTangent=-38.2052, interpMode=RCIM_Cubic, leaveTangent=-38.2052, tangentMode=RCTM_User, time=0, value=8}, <at 0.5 1>]}}}))
    ScaleA=true
    ScaleRGB=true
   )
   &CameraOffset(Camera Offset Amount=-50) {module=CameraOffset001}
  ]
  rendererClasses [&NiagaraSpriteRendererProperties]
  renderers [{rendererClass=&NiagaraSpriteRendererProperties}]
 }
 SimpleSpriteBurst {
  emitterUpdateScript [
   &EmitterState(Life Cycle Mode=::NewEnumerator1, Loop Behavior=::NewEnumerator1, Loop Duration=0.7, UseLoopDelay=false)
   &SpawnBurst_Instantaneous(Use Spawn Probability=false, UseLoopCountLimit=false)
   &SpawnBurst_Instantaneous(Spawn Group=1, Use Spawn Probability=false, UseLoopCountLimit=false) {module=SpawnBurst_Instantaneous001}
  ]
  particleSpawnScript [
   &InitializeParticle(
    Color=&MaskLinearColorBySpawnGroup(Masked Linear Color=<rgba 1 0.3954 0.5829 1>, Passthrough Linear Color=<rgba 15 5.9305 8.744 1>)
    Lifetime=0.6
    Mesh Scale Mode=::NewEnumerator3
    Sprite Size Mode=::NewEnumerator3
    Uniform Sprite Size=100
    UsePositionOffset=false
    Write Material Random=true
   )
   &InitialMeshOrientation(Mesh Axis To Orient=<0 1 0>, Mesh Orientation Mode=::NewEnumerator2, Rotation=<-0.01 0.05 -0.05>, Use Rotation Vector=true)
   /SetVariables(Particles.MeshIndex=&SpawnGroup()) {isSetParameters=true}
  ]
  particleUpdateScript [
   &ParticleState()
   &ScaleMeshSize(Scale Factor=&Multiply_VectorByFloat(
     Vector=&VectorFromCurve(VectorCurve=DataInterface() {propertyValues {
        XCurve {keys [<at 0 1>,<at 1 1>]}
        YCurve {keys [<at 0 1>,<at 1 1>]}
        ZCurve {
         defaultValue=1
         keys [
          <at 0 1 RCIM_Cubic>
          {arriveTangent=-0.1459, interpMode=RCIM_Cubic, leaveTangent=-0.1459, time=0.1379, value=1.0851}
          {arriveTangent=-0.0815, interpMode=RCIM_Cubic, leaveTangent=-0.0815, time=0.3339, value=0.9513}
          {arriveTangent=-0.0595, interpMode=RCIM_Cubic, leaveTangent=-0.0595, time=0.4821, value=1.0571}
          {arriveTangent=-0.1102, interpMode=RCIM_Cubic, leaveTangent=-0.1102, time=0.7313, value=0.9276}
          <at 1 1 RCIM_Cubic>
         ]
        }
       }})
     Float=&MakeFloatFromVector(
      Vector=&VectorFromFloat(Value=&FloatFromCurve(FloatCurve=DataInterface() {propertyValues {Curve {
           defaultValue=1
           keys [
            {arriveTangent=4.1785, interpMode=RCIM_Cubic, leaveTangent=4.1785, tangentMode=RCTM_User, time=0, value=0}
            {arriveTangent=1.1179, interpMode=RCIM_Cubic, leaveTangent=1.1179, tangentMode=RCTM_User, time=0.2042, value=0.6228}
            {arriveTangent=0.0224, interpMode=RCIM_Cubic, leaveTangent=0.0224, tangentMode=RCTM_User, time=1, value=1}
           ]
          }}}))
      Channel=::NewEnumerator0
     )
    )) {module=ScaleMeshSize001}
   &DynamicMaterialParameters(
    Index 0 Param 1=&FloatFromCurve(FloatCurve=DataInterface() {propertyValues {Curve {
        defaultValue=1
        keys [
         {arriveTangent=9.0725, interpMode=RCIM_Cubic, leaveTangent=9.0725, tangentMode=RCTM_User, time=0, value=0}
         {arriveTangent=-2.5249, interpMode=RCIM_Cubic, leaveTangent=-2.5249, tangentMode=RCTM_User, time=0.1408, value=0.9847}
         {arriveTangent=-0.0405, interpMode=RCIM_Cubic, leaveTangent=-0.0405, tangentMode=RCTM_User, time=1, value=0}
        ]
       }}})
    Index 0 Param 3=&FloatFromCurve(FloatCurve=DataInterface() {propertyValues {Curve {defaultValue=1, keys [{tangentMode=RCTM_User, time=0, value=0}, {tangentMode=RCTM_User, time=1, value=1}]}}})
    Module.Index0Float0Write=true
    Module.Index0Float1Write=false
    Module.Index0Float2Write=true
    Module.Index0Float3Write=false
   )
   &ScaleColor(
    Scale Alpha=&FloatFromCurve(FloatCurve=DataInterface() {propertyValues {Curve {
        defaultValue=0
        keys [
         {arriveTangent=9.0725, interpMode=RCIM_Cubic, leaveTangent=9.0725, tangentMode=RCTM_User, time=0, value=0}
         {arriveTangent=-2.5249, interpMode=RCIM_Cubic, leaveTangent=-2.5249, tangentMode=RCTM_User, time=0.1408, value=0.9847}
         {arriveTangent=-0.0405, interpMode=RCIM_Cubic, leaveTangent=-0.0405, tangentMode=RCTM_User, time=1, value=0}
        ]
       }}})
    Scale RGB=&VectorFromFloat(Value=&FloatFromCurve(FloatCurve=DataInterface() {propertyValues {Curve {keys [<at 0 1>, <at 0.1234 2>, {arriveTangent=-1.1407, interpMode=RCIM_Cubic, leaveTangent=-1.1407, time=0.3119, value=1.5}, <at 1 1>]}}}))
    ScaleA=true
    ScaleRGB=true
   )
   &SolveForcesAndVelocity(Clamp Velocity=false, Limit Acceleration=false)
   /SetVariables(Particles.SpriteGlowColor:LinearColor=&Multiply_LinearColorByFloat(Linear Color=$Particles.Color, Float=0.07)) {isSetParameters=true}
   &FN_Light_Attributes(
    Color=<rgba 1 0.3214 0.7112 10>
    Create Color Attribute=true
    Create Exponent Attribute=true
    Create Position Attribute=true
    Create Radius Attribute=true
    Enable Particle Light Enable=true
    Exponent=5
    Light Enabled=true
    Position:NiagaraPosition=$Particles.Position
    Radius=&FloatFromCurve(
     FloatCurve=DataInterface() {propertyValues {Curve {
        defaultValue=0
        keys [
         {arriveTangent=9.0725, interpMode=RCIM_Cubic, leaveTangent=9.0725, tangentMode=RCTM_User, time=0, value=0}
         {arriveTangent=0.2697, interpMode=RCIM_Constant, tangentMode=RCTM_User, time=0.3, value=0.9847}
         {arriveTangent=-1.4067, interpMode=RCIM_Cubic, leaveTangent=-1.4067, time=0.7, value=1}
         {arriveTangent=-0.0405, interpMode=RCIM_Cubic, leaveTangent=-0.0405, tangentMode=RCTM_User, time=1, value=0}
        ]
       }}}
     Scale Curve=75
    )
    Scale Color=&FloatFromCurve(FloatCurve=DataInterface() {propertyValues {Curve {
        defaultValue=0
        keys [
         {arriveTangent=9.0725, interpMode=RCIM_Cubic, leaveTangent=9.0725, tangentMode=RCTM_User, time=0, value=0}
         {arriveTangent=0.2697, interpMode=RCIM_Constant, tangentMode=RCTM_User, time=0.3, value=0.9847}
         {arriveTangent=-1.4067, interpMode=RCIM_Cubic, leaveTangent=-1.4067, time=0.4898, value=1}
         {arriveTangent=-0.0405, interpMode=RCIM_Cubic, leaveTangent=-0.0405, tangentMode=RCTM_User, time=1, value=0}
        ]
       }}})
   )
  ]
  rendererClasses [
   &NiagaraMeshRendererProperties
   &NiagaraSpriteRendererProperties
   &NiagaraLightRendererProperties
  ]
  renderers [
   {rendererClass=&NiagaraMeshRendererProperties, rendererIndex=0}
   {rendererClass=&NiagaraSpriteRendererProperties, rendererIndex=1}
   {rendererClass=&NiagaraLightRendererProperties, rendererIndex=2}
  ]
 }
 Sparks_and_Hearts {
  emitterUpdateScript [
   &EmitterState()
   &SpawnBurst_Instantaneous(
    Spawn Count=&RandomRangeInt(Minimum=7, Maximum=10)
    Use Spawn Probability=false
    UseLoopCountLimit=false
   )
   &SpawnBurst_Instantaneous(
    Spawn Count=&RandomRangeInt(Minimum=13, Maximum=18)
    Spawn Group=1
    Use Spawn Probability=false
    UseLoopCountLimit=false
   ) {module=SpawnBurst_Instantaneous001}
  ]
  particleSpawnScript [
   /SetVariables(Particles.VisibilityTag=&SpawnGroup()) {isSetParameters=true}
   &InitializeParticle(
    AdjustAlpha=false
    AdjustHue=true
    AdjustSaturation=false
    AdjustValue=true
    Color=&MaskLinearColorBySpawnGroup(Masked Linear Color=<rgba 15 1.6424 7.108 2>, Passthrough Linear Color=<rgba 1 0.8148 0.2831 1>)
    Color Mode=::NewEnumerator3
    Hue Shift Range:Vector2f=<-0.05 0.05>
    Lifetime Mode=::NewEnumerator1
    Sprite Size Mode=::NewEnumerator3
    Sprite UV Mode=::NewEnumerator1
    Uniform Sprite Size=&MaskFloatBySpawnGroup(
     Masked Float=&RandomRangeFloat(Minimum=10, Maximum=25)
     Passthrough Float=&RandomRangeFloat(Minimum=5, Maximum=35)
    )
    UsePositionOffset=false
    Value Range:Vector2f=<0.9 2>
    Write Material Random=true
   )
   &AddVelocity(
    Module.Use Falloff Distance=false
    Origin Offset=&RandomRangeVector(Minimum=<-0.1 -0.1 -0.1>, Maximum=<0.1 0.1 0.1>)
    Override Seed=false
    Velocity Mode=::NewEnumerator1
    Velocity Speed=&RandomRangeFloat(Minimum=25, Maximum=35)
   )
   &SubUVAnimation(
    Evaluation Type=::NewEnumerator0
    "Sprite Renderer "=DataInterface() {propertyValues {SpriteRenderer=&self:Sparks_and_Hearts_0.NiagaraSpriteRendererProperties_0}}
    SubUV Animation Mode:ENiagara_SubUVLookupModeV2=::NewEnumerator2
    UseEndFrame=false
    UseStartFrame=false
   ) {module=SubUVAnimation001}
   &ShapeLocation(MaskSpawnGroup=false, Override Seed=false, Sphere Radius=5, UseOwnerScale=true)
  ]
  particleUpdateScript [
   &ParticleState()
   &SpriteRotationRate(Rotation Rate=&FloatFromCurve(
     FloatCurve=DataInterface() {propertyValues {Curve {
        defaultValue=0
        keys [
         {arriveTangent=1.8448, interpMode=RCIM_Cubic, leaveTangent=1.8448, tangentMode=RCTM_User, time=0, value=0}
         {interpMode=RCIM_Cubic, tangentMode=RCTM_User, time=1, value=1}
        ]
       }}}
     Scale Curve=&MaskFloatBySpawnGroup(
      Masked Float=&RandomRangeFloat(Minimum=-200, Maximum=200)
      Passthrough Float=&RandomRangeFloat(Minimum=-500, Maximum=500)
     )
    ))
   &ScaleSpriteSize(
    Scale Factor=/Niagara/DynamicInputs/Helpers/MaskVector2d(
     Mask Source:NiagaraInt32=0
     Mask Parameter:NiagaraInt32=?
     Mask:NiagaraInt32=0
     Masked Vector2D:Vector2f=&Vector2DFromCurve(
      Vector2Curve=DataInterface() {propertyValues {
        XCurve {keys [{tangentMode=RCTM_User, time=0, value=1}, {tangentMode=RCTM_User, time=1, value=1}]}
        YCurve {keys [
          <at 0 1 RCIM_Cubic>
          {arriveTangent=-3.1701, interpMode=RCIM_Cubic, leaveTangent=-3.1701, time=0.0398, value=2.1}
          {arriveTangent=-3.0843, interpMode=RCIM_Cubic, leaveTangent=-3.0843, time=0.1104, value=0.65}
          {arriveTangent=0.2979, interpMode=RCIM_Cubic, leaveTangent=0.2979, tangentMode=RCTM_User, time=0.1576, value=1.7365}
          {arriveTangent=-2.338, interpMode=RCIM_Cubic, leaveTangent=-2.338, time=0.2071, value=0.8282}
          {arriveTangent=1.4742, interpMode=RCIM_Cubic, leaveTangent=1.4742, time=0.286, value=1.4365}
          {arriveTangent=-1.4211, interpMode=RCIM_Cubic, leaveTangent=-1.4211, time=0.3237, value=1}
          {arriveTangent=-0.8335, interpMode=RCIM_Cubic, leaveTangent=-0.8335, time=0.3909, value=1.2873}
          {arriveTangent=-0.5938, interpMode=RCIM_Cubic, leaveTangent=-0.5938, time=0.5301, value=0.8279}
          {arriveTangent=0.3663, interpMode=RCIM_Cubic, leaveTangent=0.3663, time=0.8748, value=1}
          <at 1 1 RCIM_Cubic>
         ]}
       }}
      ScaleCurve=<1 1>
     )
     Passthrough Vector2D:Vector2f=<1 1>
    )
    Scale Sprite Size Mode=::NewEnumerator2
   )
   &ScaleSpriteSize(
    Scale Sprite Size Mode=::NewEnumerator0
    Uniform Scale Factor=&FloatFromCurve(FloatCurve=DataInterface() {propertyValues {Curve {
        defaultValue=0
        keys [
         {arriveTangent=9.0725, interpMode=RCIM_Cubic, leaveTangent=9.0725, tangentMode=RCTM_User, time=0, value=0}
         {arriveTangent=-2.5249, interpMode=RCIM_Cubic, leaveTangent=-2.5249, tangentMode=RCTM_User, time=0.1408, value=0.9847}
         {arriveTangent=-0.0405, interpMode=RCIM_Cubic, leaveTangent=-0.0405, tangentMode=RCTM_User, time=1, value=0}
        ]
       }}})
   ) {module=ScaleSpriteSize001}
   &PointForce(Force Strength=1000, Module.Use Falloff=true, Module.UseLinearForce=false)
   &CurlNoiseForce(Bias Noise=false, Noise Frequency=15, Noise Strength=200, Pan Noise=true, Pan Noise Field=<0 0 1>)
   &Drag(
    Drag=&FloatFromCurve(
     FloatCurve=DataInterface() {propertyValues {Curve {
        defaultValue=0
        keys [
         {arriveTangent=-0.1861, interpMode=RCIM_Cubic, leaveTangent=-0.1861, tangentMode=RCTM_User, time=0, value=0}
         {arriveTangent=-1.5452, interpMode=RCIM_Cubic, leaveTangent=-1.5452, tangentMode=RCTM_User, time=0.4054, value=0.8656}
         {arriveTangent=-0.0405, interpMode=RCIM_Cubic, leaveTangent=-0.0405, tangentMode=RCTM_User, time=1, value=0.3}
        ]
       }}}
     Scale Curve=&RandomRangeFloat(Minimum=15, Maximum=35)
    )
    Use Linear Drag=true
    Use Rotational Drag=false
   )
   &GravityForce(Gravity=&VectorFromCurve(
     VectorCurve=DataInterface() {propertyValues {
       XCurve {keys [<at 0 1>,<at 1 1>]}
       YCurve {keys [<at 0 1>,<at 1 1>]}
       ZCurve {
        defaultValue=1
        keys [{interpMode=RCIM_Cubic, tangentMode=RCTM_User, time=0, value=0}, {interpMode=RCIM_Cubic, tangentMode=RCTM_User, time=1, value=1}]
       }
      }}
     Scale Curve=<0 0 980>
    ))
   &ScaleColor(
    Scale Alpha=&MaskFloatBySpawnGroup(
     Masked Float=&FloatFromCurve(FloatCurve=DataInterface() {propertyValues {Curve {
         defaultValue=1.0113
         keys [
          {arriveTangent=9.0725, interpMode=RCIM_Cubic, leaveTangent=9.0725, tangentMode=RCTM_User, time=0, value=0}
          {arriveTangent=-2.5249, interpMode=RCIM_Cubic, leaveTangent=-2.5249, tangentMode=RCTM_User, time=0.1408, value=0.9847}
          {arriveTangent=-0.0405, interpMode=RCIM_Cubic, leaveTangent=-0.0405, tangentMode=RCTM_User, time=1, value=0}
         ]
        }}})
     Passthrough Float=&FloatFromCurve(FloatCurve=DataInterface() {propertyValues {Curve {
         defaultValue=1.0113
         keys [
          {arriveTangent=9.0725, interpMode=RCIM_Cubic, leaveTangent=9.0725, tangentMode=RCTM_User, time=0, value=0}
          {arriveTangent=-2.5249, interpMode=RCIM_Cubic, leaveTangent=-2.5249, tangentMode=RCTM_User, time=0.1408, value=0.9847}
          {arriveTangent=-0.2598, interpMode=RCIM_Cubic, leaveTangent=-0.2598, time=0.427, value=0.334}
          {arriveTangent=-1.024, interpMode=RCIM_Cubic, leaveTangent=-1.024, time=0.4307, value=0.9093}
          {arriveTangent=-1.1872, interpMode=RCIM_Cubic, leaveTangent=-1.1872, time=0.5924, value=0.1646}
          {arriveTangent=-0.3456, interpMode=RCIM_Cubic, leaveTangent=-0.3456, time=0.607, value=0.7}
          {arriveTangent=-1.7812, interpMode=RCIM_Cubic, leaveTangent=-1.7812, time=0.797, value=0.0939}
          {arriveTangent=-0.0405, interpMode=RCIM_Cubic, leaveTangent=-0.0405, tangentMode=RCTM_User, time=1, value=0}
         ]
        }}})
    )
    ScaleA=true
    ScaleRGB=true
   )
   &SolveForcesAndVelocity(Clamp Velocity=false, Limit Acceleration=false)
   /SetVariables(
    Particles.GlowCamOffset=-5
    Particles.GlowSize=&Multiply_Vector2D_ByFloat(Float=&RandomRangeFloat(Minimum=2, Maximum=4), Vector2D=$Particles.SpriteSize)
    Particles.SpriteGlowColor:LinearColor=&Multiply_LinearColorByFloat(
     Linear Color=$Particles.Color
     Float=&Multiply_Float(
      A=&FloatFromCurve(FloatCurve=DataInterface() {propertyValues {Curve {
          defaultValue=1.0113
          keys [
           {arriveTangent=9.0725, interpMode=RCIM_Cubic, leaveTangent=9.0725, tangentMode=RCTM_User, time=0, value=0}
           {arriveTangent=-2.5249, interpMode=RCIM_Cubic, leaveTangent=-2.5249, tangentMode=RCTM_User, time=0.1408, value=0.9847}
           {arriveTangent=-0.2598, interpMode=RCIM_Cubic, leaveTangent=-0.2598, time=0.427, value=0.334}
           {arriveTangent=-1.024, interpMode=RCIM_Cubic, leaveTangent=-1.024, time=0.4307, value=0.9093}
           {arriveTangent=-1.1872, interpMode=RCIM_Cubic, leaveTangent=-1.1872, time=0.5924, value=0.1646}
           {arriveTangent=-0.3456, interpMode=RCIM_Cubic, leaveTangent=-0.3456, time=0.607, value=0.7}
           {arriveTangent=-1.7812, interpMode=RCIM_Cubic, leaveTangent=-1.7812, time=0.797, value=0.0939}
           {arriveTangent=-0.0405, interpMode=RCIM_Cubic, leaveTangent=-0.0405, tangentMode=RCTM_User, time=1, value=0}
          ]
         }}})
      B=&RandomRangeFloat(Minimum=0.1, Maximum=0.4)
     )
    )
   ) {isSetParameters=true}
  ]
  rendererClasses [&NiagaraSpriteRendererProperties]
  renderers [
   {rendererClass=&NiagaraSpriteRendererProperties}
   {rendererClass=&NiagaraSpriteRendererProperties, rendererIndex=1}
   {rendererClass=&NiagaraSpriteRendererProperties, rendererIndex=2}
  ]
 }
}
summary {
 systemName=NS_Backplates_Heart
 emitters [
  {emitterName=Sparks_and_Hearts, rendererClasses [&NiagaraSpriteRendererProperties]}
  {
   emitterName=SimpleSpriteBurst
   rendererClasses [
    &NiagaraMeshRendererProperties
    &NiagaraSpriteRendererProperties
    &NiagaraLightRendererProperties
   ]
  }
  {emitterName=Rays, rendererClasses [&NiagaraSpriteRendererProperties]}
 ]
 userVariables []
}
