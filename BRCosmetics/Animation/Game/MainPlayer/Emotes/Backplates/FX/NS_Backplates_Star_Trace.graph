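# Niagara effect for a star backplate in an emote, apparently tracing the star on in staggered stages, with glows and twinkles.
# Star_Point bursts mesh particles by index, using per-index curves and masks to animate material parameters, outline colors and sort order from shared system star values.
# Boost adds flare sprites, Little_Stars twinkles at five preset points, and Light pulses blue.
# Separate Star_Body, Upper_Lines and Lower_Lines emitters are disabled, their curves largely mirrored in Star_Point.
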NiagaraSystem /BRCosmetics/Animation/Game/MainPlayer/Emotes/Backplates/FX/NS_Backplates_Star_Trace
dependencies {
 usedDataInterfaces [
  &NiagaraDataInterfaceCurve
  &NiagaraDataInterfaceArrayInt32
  &NiagaraDataInterfaceMeshRendererInfo
  &NiagaraDataInterfaceArrayFloat
  &NiagaraDataInterfaceArrayFloat3
 ]
 usedDynamicInputs []
 usedModules [
  &SystemState
  &self:SystemSpawnScript.SystemScriptSource.SystemScriptGraph.NiagaraNodeAssignment_1.SetVariables_#1
 ]
}
emitters {
 Boost {
  emitterUpdateScript [
   &EmitterState(Inactive Response=::NewEnumerator1, Life Cycle Mode=::NewEnumerator1, Loop Behavior=::NewEnumerator1, Loop Duration=2.5, UseLoopDelay=false)
   &SpawnBurst_Instantaneous(
    Spawn Count=2
    Spawn Time=&MaskFloat(Mask Source=1, Mask Parameter=&ReturnExecIndex(), Mask=1, Masked Float=0, Passthrough Float=1.3)
    Use Spawn Probability=false
    UseLoopCountLimit=false
   )
  ]
  particleSpawnScript [
   /SetVariables(Particles.VisibilityTag=$Engine.ExecIndex) {isSetParameters=true}
   &InitializeParticle(
    Color=<rgba 0 0.5535 5 1>
    Position Offset=<1 0 5>
    Sprite Size Mode=::NewEnumerator3
    Uniform Sprite Size=128
    UsePositionOffset=true
    Write Material Random=true
   )
   &ApplyInitialForces()
   &AlignSpriteToMeshOrientation(
    Mesh Orientation Relative Sprite Alignment Vector=<0 0 1>
    Mesh Orientation Relative Sprite Facing Vector:Vector3f=<0 -1 0>
    SetAlignment=true
    SetFacing=true
   )
  ]
  particleUpdateScript [
   &ParticleState()
   &ScaleColor(
    Scale Alpha=&FloatFromCurve(FloatCurve=DataInterface() {propertyValues {Curve {
        defaultValue=0
        keys [
         {arriveTangent=3.4683, time=0, value=0}
         {arriveTangent=2.5, interpMode=RCIM_Cubic, leaveTangent=2.5, time=0.1234, value=0}
         {arriveTangent=5.6231, interpMode=RCIM_Cubic, leaveTangent=-2.2018, tangentMode=RCTM_Break, time=0.4, value=1}
         {tangentMode=RCTM_User, time=1, value=0}
        ]
       }}})
    Scale RGB=&VectorFromFloat(Value=&FloatFromCurve(FloatCurve=DataInterface() {propertyValues {Curve {keys [
          {arriveTangent=-2.9773, interpMode=RCIM_Cubic, leaveTangent=-2.9773, tangentMode=RCTM_User, time=0, value=1}
          {arriveTangent=12.0199, interpMode=RCIM_Cubic, leaveTangent=-4.2217, tangentMode=RCTM_Break, time=0.3579, value=2.4118}
          {arriveTangent=-2.1987, interpMode=RCIM_Cubic, leaveTangent=-2.1987, time=0.6465, value=0.7843}
          <at 1 1>
         ]}}}))
    ScaleA=true
    ScaleRGB=true
   )
   &SolveForcesAndVelocity(Clamp Velocity=false, Limit Acceleration=false)
   /SetVariables(
    Particles.GlowCamOffset=-10
    Particles.GlowSize=&Multiply_Vector2D_ByFloat(
     Float=&FloatFromCurve(
      FloatCurve=DataInterface() {propertyValues {Curve {
         defaultValue=0
         keys [
          {interpMode=RCIM_Cubic, tangentMode=RCTM_User, time=0, value=0}
          {interpMode=RCIM_Cubic, tangentMode=RCTM_User, time=0.5, value=1}
          {tangentMode=RCTM_User, time=1, value=0}
         ]
        }}}
      Scale Curve=2
     )
     Vector2D=$Particles.SpriteSize
    )
    Particles.LensFlareColor:LinearColor=&Multiply_LinearColorByFloat(
     Linear Color=$System.StarColor
     Float=&FloatFromCurve(FloatCurve=DataInterface() {propertyValues {Curve {
         defaultValue=0
         keys [
          {interpMode=RCIM_Cubic, tangentMode=RCTM_User, time=0, value=0}
          {interpMode=RCIM_Cubic, tangentMode=RCTM_User, time=0.5, value=1}
          {tangentMode=RCTM_User, time=1, value=0}
         ]
        }}})
    )
    Particles.LensSize:Vector2f=&Multiply_Vector2D_ByFloat(
     Float=&FloatFromCurve(
      FloatCurve=DataInterface() {propertyValues {Curve {
         defaultValue=0
         keys [
          {arriveTangent=4.1785, interpMode=RCIM_Cubic, leaveTangent=4.1785, tangentMode=RCTM_User, time=0, value=0}
          {arriveTangent=1.1179, interpMode=RCIM_Cubic, leaveTangent=1.1179, tangentMode=RCTM_User, time=0.2042, value=0.6228}
          {arriveTangent=0.0224, interpMode=RCIM_Cubic, leaveTangent=0.0224, tangentMode=RCTM_User, time=1, value=1}
         ]
        }}}
      Scale Curve=1.75
     )
     Vector2D=$Particles.SpriteSize
    )
    Particles.SpriteGlowColor:LinearColor=&Multiply_LinearColorByFloat(
     Linear Color=$System.StarColor
     Float=&FloatFromCurve(
      FloatCurve=DataInterface() {propertyValues {Curve {
         defaultValue=0
         keys [
          {interpMode=RCIM_Cubic, tangentMode=RCTM_User, time=0, value=0}
          {interpMode=RCIM_Cubic, tangentMode=RCTM_User, time=0.5, value=1}
          {tangentMode=RCTM_User, time=1, value=0}
         ]
        }}}
      Scale Curve=0.2
     )
    )
   ) {isSetParameters=true}
  ]
  rendererClasses [&NiagaraSpriteRendererProperties]
  renderers [
   {rendererClass=&NiagaraSpriteRendererProperties}
   {rendererClass=&NiagaraSpriteRendererProperties, rendererIndex=1}
   {rendererClass=&NiagaraSpriteRendererProperties, rendererIndex=2}
  ]
 }
 Light {
  emitterUpdateScript [
   &EmitterState(Life Cycle Mode=::NewEnumerator1, Loop Behavior=::NewEnumerator1, Loop Duration=3.5, UseLoopDelay=false)
   &SpawnBurst_Instantaneous(Use Spawn Probability=false, UseLoopCountLimit=false)
  ]
  particleSpawnScript [&InitializeParticle(Color Mode=::NewEnumerator0, Lifetime=3.5, UsePositionOffset=false, Write Material Random=true)]
  particleUpdateScript [
   &ParticleState()
   &FN_Light_Attributes(
    Color=<rgba 0.0002 0.2958 1 20>
    Create Color Attribute=true
    Create Exponent Attribute=true
    Create Position Attribute=true
    Create Radius Attribute=true
    Enable Particle Light Enable=true
    Exponent=5
    Light Enabled=true
    Position:NiagaraPosition=$Particles.Position
    Radius=&FloatFromCurve(
     FloatCurve=DataInterface() {propertyValues {Curve {
        defaultValue=0
        keys [
         {arriveTangent=9.0725, interpMode=RCIM_Cubic, leaveTangent=9.0725, tangentMode=RCTM_User, time=0, value=0}
         {arriveTangent=0.2697, interpMode=RCIM_Constant, tangentMode=RCTM_User, time=0.3, value=0.9847}
         {arriveTangent=-1.4067, interpMode=RCIM_Cubic, leaveTangent=-1.4067, time=0.7, value=1}
         {arriveTangent=-0.0405, interpMode=RCIM_Cubic, leaveTangent=-0.0405, tangentMode=RCTM_User, time=1, value=0}
        ]
       }}}
     Scale Curve=75
    )
    Scale Color=&FloatFromCurve(FloatCurve=DataInterface() {propertyValues {Curve {
        defaultValue=0
        keys [
         {arriveTangent=9.0725, interpMode=RCIM_Cubic, leaveTangent=9.0725, tangentMode=RCTM_User, time=0, value=0}
         {arriveTangent=0.2697, interpMode=RCIM_Constant, tangentMode=RCTM_User, time=0.3, value=0.9847}
         {arriveTangent=-1.4067, interpMode=RCIM_Cubic, leaveTangent=-1.4067, time=0.4898, value=1}
         {arriveTangent=-0.0405, interpMode=RCIM_Cubic, leaveTangent=-0.0405, tangentMode=RCTM_User, time=1, value=0}
        ]
       }}})
   )
  ]
  rendererClasses [&NiagaraLightRendererProperties]
  renderers [{rendererClass=&NiagaraLightRendererProperties, rendererIndex=0}]
 }
 Little_Stars {
  emitterUpdateScript [
   &EmitterState(Life Cycle Mode=::NewEnumerator1, Loop Behavior=::NewEnumerator1, Loop Duration=1.7, UseLoopDelay=false)
   &SpawnRate(SpawnRate=6.5, Use Spawn Probability=false)
  ]
  particleSpawnScript [
   &InitializeParticle(
    Color=$System.StarColor
    Sprite Size Mode=::NewEnumerator1
    Uniform Sprite Size Max=30
    Uniform Sprite Size Min=20
    UsePositionOffset=false
    Write Material Random=true
   )
   &PartitionParticles(Number Of Partitions=5, Particles Per Partition=1, Partition Mode:ENiagara_PartitionMode_v2=::NewEnumerator1)
   &ApplyInitialForces()
   &InitialMeshOrientation(
    Mesh Orientation Mode=::NewEnumerator2
    Rotation=&SelectVectorFromArray(
     Array Sampling Mode=::NewEnumerator0
     Direct Array Index=?
     Vector Selection Array=DataInterface() {propertyValues {FloatData [<0 -0.05 0>, <0 0.05 0>, <0 0 0>]}}
     Random Seed=0
    )
    Use Rotation Vector=true
   )
   &AlignSpriteToMeshOrientation(
    Mesh Orientation Relative Sprite Alignment Vector=<0 0 1>
    Mesh Orientation Relative Sprite Facing Vector:Vector3f=<0 -1 0>
    SetAlignment=true
    SetFacing=true
   )
   /SetVariables(Particles.Position=&SelectVectorFromArray(
     Direct Array Index=$Particles.PartitionParticles.Partition
     Interpolation=&FloatFromCurve(FloatCurve=DataInterface() {propertyValues {Curve {keys [<at 0 1>,<at 1 0>]}}})
     Vector Selection Array=DataInterface() {propertyValues {FloatData [<-25 0 65>, <50 0 40>, <-55 0 -10>, <55 0 -25>, <10 0 -65>]}}
    )) {isSetParameters=true}
  ]
  particleUpdateScript [
   &ParticleState()
   &ScaleSpriteSize(
    Uniform Curve Scale=&RandomRangeFloat(Minimum=0.9, Maximum=1.1)
    Uniform Curve Sprite Scale=DataInterface() {propertyValues {Curve {
       defaultValue=0
       keys [
        {arriveTangent=9.0725, interpMode=RCIM_Cubic, leaveTangent=9.0725, tangentMode=RCTM_User, time=0, value=0}
        {arriveTangent=-2.5249, interpMode=RCIM_Cubic, leaveTangent=-2.5249, tangentMode=RCTM_User, time=0.1408, value=0.9847}
        {arriveTangent=-0.0405, interpMode=RCIM_Cubic, leaveTangent=-0.0405, tangentMode=RCTM_User, time=1, value=0}
       ]
      }}}
   )
   &ScaleColor(
    Scale Alpha=&FloatFromCurve(FloatCurve=DataInterface() {propertyValues {Curve {
        defaultValue=1
        keys [
         {arriveTangent=9.0725, interpMode=RCIM_Cubic, leaveTangent=9.0725, tangentMode=RCTM_User, time=0, value=0}
         {arriveTangent=-2.5249, interpMode=RCIM_Cubic, leaveTangent=-2.5249, tangentMode=RCTM_User, time=0.1408, value=0.9847}
         {arriveTangent=-0.0405, interpMode=RCIM_Cubic, leaveTangent=-0.0405, tangentMode=RCTM_User, time=1, value=0}
        ]
       }}})
    Scale RGB=&VectorFromFloat(Value=&FloatFromCurve(
      FloatCurve=DataInterface() {propertyValues {Curve {
         defaultValue=1
         keys [
          {arriveTangent=9.0725, interpMode=RCIM_Cubic, leaveTangent=9.0725, tangentMode=RCTM_User, time=0, value=1}
          {arriveTangent=-2.5249, interpMode=RCIM_Cubic, leaveTangent=-2.5249, tangentMode=RCTM_User, time=0.1408, value=2}
          {arriveTangent=-0.0405, interpMode=RCIM_Cubic, leaveTangent=-0.0405, tangentMode=RCTM_User, time=1, value=1}
         ]
        }}}
      Scale Curve=5
     ))
    ScaleA=true
    ScaleRGB=true
   )
   /SetVariables(
    Particles.GlowCamOffset=-10
    Particles.LensFlareColor:LinearColor=&Multiply_LinearColorByFloat(
     Linear Color=<rgba 0.2796 0.3492 1.3 1>
     Float=&FloatFromCurve(
      FloatCurve=DataInterface() {propertyValues {Curve {
         defaultValue=0
         keys [
          {arriveTangent=9.0725, interpMode=RCIM_Cubic, leaveTangent=9.0725, tangentMode=RCTM_User, time=0, value=0}
          {arriveTangent=-2.5249, interpMode=RCIM_Cubic, leaveTangent=-2.5249, tangentMode=RCTM_User, time=0.1408, value=0.9847}
          {arriveTangent=-0.0405, interpMode=RCIM_Cubic, leaveTangent=-0.0405, tangentMode=RCTM_User, time=1, value=0}
         ]
        }}}
      Scale Curve=0.4
     )
    )
    Particles.LensSize:Vector2f=&Multiply_Vector2D_ByFloat(
     Float=&FloatFromCurve(FloatCurve=DataInterface() {propertyValues {Curve {defaultValue=1, keys [{tangentMode=RCTM_User, time=0, value=1}, {tangentMode=RCTM_User, time=1, value=3}]}}})
     Vector2D=&Multiply_Vector2D_ByFloat(Float=35, Vector2D=<1 1>)
    )
   ) {isSetParameters=true}
  ]
  rendererClasses [&NiagaraSpriteRendererProperties]
  renderers [
   {rendererClass=&NiagaraSpriteRendererProperties}
   {rendererClass=&NiagaraSpriteRendererProperties, rendererIndex=1}
  ]
 }
 Star_Body {
  bEnabled=false
  emitterUpdateScript [
   &EmitterState()
   &SpawnBurst_Instantaneous(Spawn Count=1, Spawn Time=0.5, Use Spawn Probability=false)
  ]
  particleSpawnScript [&InitializeParticle(
    Color=$System.StarColor
    Lifetime=3.5
    Mesh Scale Mode=::NewEnumerator3
    Mesh Uniform Scale=$System.StarScale
    UsePositionOffset=false
    Write Material Random=true
   ) {module=InitializeParticle001}]
  particleUpdateScript [
   &ParticleState()
   &ScaleMeshSize(Scale Factor=&VectorFromFloat(Value=&FloatFromCurve(FloatCurve=DataInterface() {propertyValues {Curve {
         defaultValue=1
         keys [
          {tangentMode=RCTM_User, time=0, value=1}
          {arriveTangent=-0.1707, interpMode=RCIM_Cubic, leaveTangent=-0.1707, tangentMode=RCTM_User, time=0.4969, value=1}
          {arriveTangent=2.4771, interpMode=RCIM_Cubic, leaveTangent=2.4771, tangentMode=RCTM_User, time=0.8, value=1.3}
         ]
        }}})))
   &DynamicMaterialParameters(
    Index 0 Param 2=&FloatFromCurve(FloatCurve=DataInterface() {propertyValues {Curve {
        defaultValue=1
        keys [
         {arriveTangent=1.3832, interpMode=RCIM_Cubic, leaveTangent=1.3832, tangentMode=RCTM_User, time=0, value=0}
         {arriveTangent=5.9809, interpMode=RCIM_Cubic, tangentMode=RCTM_Break, time=0.1429, value=0.65}
         {arriveTangent=3.149, tangentMode=RCTM_User, time=0.215, value=0.8}
         {tangentMode=RCTM_User, time=0.24, value=1}
         {arriveTangent=-1.5975, interpMode=RCIM_Cubic, leaveTangent=-1.5975, time=0.4, value=2}
         {arriveTangent=-3.3333, interpMode=RCIM_Cubic, leaveTangent=-3.3333, time=0.5034, value=0.5792}
         <at 1 0>
        ]
       }}})
    Index 0 Param 3=1
    Index 1 Param 1=$System.StarBrightness
    Module.Index0Float0Write=false
    Module.Index0Float1Write=true
    Module.Index0Float2Write=true
    Module.Index0Float3Write=false
    Module.Index1Float0Write=true
    Module.Index1Float1Write=false
    Module.Index1Float2Write=false
    Module.Index1Float3Write=false
    Write Parameter Index 1=::NewEnumerator1
   )
   &SolveForcesAndVelocity(Clamp Velocity=false, Limit Acceleration=false)
   &ScaleColor(
    Scale Alpha=&FloatFromCurve(FloatCurve=DataInterface() {propertyValues {Curve {
        defaultValue=1
        keys [
         {arriveTangent=9.0725, interpMode=RCIM_Cubic, leaveTangent=9.0725, tangentMode=RCTM_User, time=0, value=0}
         {arriveTangent=-1.0647, interpMode=RCIM_Cubic, leaveTangent=-1.0647, tangentMode=RCTM_User, time=0.1408, value=0.9847}
         {arriveTangent=-0.0405, interpMode=RCIM_Cubic, leaveTangent=-0.0405, tangentMode=RCTM_User, time=0.65, value=0}
        ]
       }}})
    ScaleA=true
    ScaleRGB=false
   )
   /SetVariables(
    Particles.DMP1:Vector4f=&MakeVector4(X=2, Y=0, Z=0, W=0)
    Particles.OutlineColor:LinearColor=&Multiply_LinearColorByFloat(
     Linear Color=$System.StarColor
     Float=&Multiply_Float(
      A=$System.OutlineIntensity
      B=&FloatFromCurve(FloatCurve=DataInterface() {propertyValues {Curve {
          defaultValue=0
          keys [
           <at 0 0>
           {arriveTangent=2.5, interpMode=RCIM_Cubic, leaveTangent=2.5, time=0.16, value=0}
           <at 0.4 1>
           {arriveTangent=-4.1167, interpMode=RCIM_Cubic, leaveTangent=-4.1167, time=0.4569, value=0.9882}
           {arriveTangent=-6.591, interpMode=RCIM_Cubic, leaveTangent=-6.591, tangentMode=RCTM_User, time=0.5151, value=0.5261}
           {tangentMode=RCTM_User, time=0.6102, value=0}
          ]
         }}})
     )
    )
   ) {isSetParameters=true}
  ]
  rendererClasses [&NiagaraMeshRendererProperties]
  renderers [
   {rendererClass=&NiagaraMeshRendererProperties, rendererIndex=0}
   {rendererClass=&NiagaraMeshRendererProperties, rendererIndex=1}
  ]
 }
 Star_Lower_Lines {
  bEnabled=false
  emitterUpdateScript [
   &EmitterState()
   &SpawnBurst_Instantaneous(Spawn Count=1, Spawn Time=1.4, Use Spawn Probability=false)
  ]
  particleSpawnScript [&InitializeParticle(
    Color=$System.StarColor
    Lifetime=3.5
    Mesh Scale Mode=::NewEnumerator3
    Mesh Uniform Scale=$System.StarScale
    UsePositionOffset=false
    Write Material Random=true
   ) {module=InitializeParticle001}]
  particleUpdateScript [
   &ParticleState()
   &ScaleMeshSize(Scale Factor=&VectorFromFloat(Value=&FloatFromCurve(FloatCurve=DataInterface() {propertyValues {Curve {
         defaultValue=1
         keys [
          {tangentMode=RCTM_User, time=0, value=1}
          {arriveTangent=-0.1707, interpMode=RCIM_Cubic, leaveTangent=-0.1707, tangentMode=RCTM_User, time=0.25, value=1}
          {arriveTangent=4.5319, interpMode=RCIM_Cubic, leaveTangent=4.5319, tangentMode=RCTM_User, time=0.45, value=1.3}
         ]
        }}})))
   &DynamicMaterialParameters(
    Index 0 Param 2=&FloatFromCurve(FloatCurve=DataInterface() {propertyValues {Curve {
        defaultValue=1
        keys [
         {arriveTangent=1.5932, interpMode=RCIM_Cubic, leaveTangent=1.5932, tangentMode=RCTM_User, time=0, value=0}
         {arriveTangent=1.4091, leaveTangent=-0.3339, tangentMode=RCTM_User, time=0.137, value=1}
         {tangentMode=RCTM_User, time=0.15, value=2}
         {arriveTangent=-2.1186, interpMode=RCIM_Cubic, leaveTangent=-1.1582, tangentMode=RCTM_Break, time=0.2862, value=0.1614}
         <at 1 0>
        ]
       }}})
    Index 0 Param 3=1
    Index 1 Param 1=$System.StarBrightness
    Module.Index0Float0Write=false
    Module.Index0Float1Write=true
    Module.Index0Float2Write=true
    Module.Index0Float3Write=false
    Module.Index1Float0Write=true
    Module.Index1Float1Write=false
    Module.Index1Float2Write=false
    Module.Index1Float3Write=false
    Write Parameter Index 1=::NewEnumerator1
   )
   &SolveForcesAndVelocity(Clamp Velocity=false, Limit Acceleration=false)
   &ScaleColor(
    Scale Alpha=&FloatFromCurve(
     FloatCurve=DataInterface() {propertyValues {Curve {
        defaultValue=1
        keys [
         {arriveTangent=9.0725, interpMode=RCIM_Cubic, leaveTangent=9.0725, tangentMode=RCTM_User, time=0, value=0}
         {arriveTangent=-1.9928, interpMode=RCIM_Cubic, leaveTangent=-1.9928, tangentMode=RCTM_User, time=0.1022, value=0.3369}
         {arriveTangent=-0.0405, interpMode=RCIM_Cubic, leaveTangent=-0.0405, tangentMode=RCTM_User, time=0.4, value=0}
        ]
       }}}
     Scale Curve=0.5
    )
    ScaleA=true
    ScaleRGB=false
   )
   /SetVariables(
    Particles.DMP1:Vector4f=&MakeVector4(X=3, Y=0, Z=0, W=0)
    Particles.OutlineColor:LinearColor=&Multiply_LinearColorByFloat(
     Linear Color=$System.StarColor
     Float=&Multiply_Float(
      A=$System.OutlineIntensity
      B=&FloatFromCurve(FloatCurve=DataInterface() {propertyValues {Curve {
          defaultValue=0
          keys [
           {arriveTangent=9.0725, interpMode=RCIM_Cubic, leaveTangent=9.0725, tangentMode=RCTM_User, time=0, value=0}
           {arriveTangent=-2.5249, leaveTangent=-2.5249, time=0.1408, value=0.9847}
           {arriveTangent=-4.3337, interpMode=RCIM_Cubic, leaveTangent=-4.3337, time=0.2018, value=1}
           {arriveTangent=-5.0443, interpMode=RCIM_Cubic, leaveTangent=-5.0443, time=0.2797, value=0.3826}
           {arriveTangent=-0.0405, interpMode=RCIM_Cubic, leaveTangent=-0.0405, tangentMode=RCTM_User, time=0.4, value=0}
          ]
         }}})
     )
    )
   ) {isSetParameters=true}
  ]
  rendererClasses [&NiagaraMeshRendererProperties]
  renderers [
   {rendererClass=&NiagaraMeshRendererProperties, rendererIndex=0}
   {rendererClass=&NiagaraMeshRendererProperties, rendererIndex=1}
  ]
 }
 Star_Point {
  emitterUpdateScript [
   &EmitterState()
   &SpawnBurst_Instantaneous(Spawn Count=7, Spawn Group:NiagaraInt32=&ReturnExecIndex(), Use Spawn Probability=false)
  ]
  particleSpawnScript [&InitializeParticle(
    Color=$System.StarColor
    Lifetime=3.5
    Mesh Renderer Array Visibility Mode=::NewEnumerator1
    Mesh Renderer Info=DataInterface() {propertyValues {MeshRenderer=&self:Star_Point_0.NiagaraMeshRendererProperties_1}}
    Mesh Renderer Mesh Index:NiagaraInt32=&SelectIntFromArray(
     Direct Array Index=&ReturnExecIndex()
     Int Selection Array=DataInterface() {propertyValues {IntData [0,1,2,3,4,5,6,7]}}
    )
    Mesh Scale Mode=::NewEnumerator3
    Mesh Uniform Scale=$System.StarScale
    UsePositionOffset=false
    Write Material Random=true
   ) {module=InitializeParticle001}]
  particleUpdateScript [
   &ParticleState()
   &ScaleMeshSize(Scale Factor=&VectorFromFloat(Value=&FloatFromCurve(FloatCurve=DataInterface() {propertyValues {Curve {
         defaultValue=1
         keys [
          {tangentMode=RCTM_User, time=0, value=1}
          {arriveTangent=-0.1707, interpMode=RCIM_Cubic, leaveTangent=-0.1707, tangentMode=RCTM_User, time=0.4969, value=1}
          {arriveTangent=1.6236, interpMode=RCIM_Cubic, leaveTangent=1.6236, tangentMode=RCTM_User, time=1, value=1.3}
         ]
        }}})))
   &SolveForcesAndVelocity(Clamp Velocity=false, Limit Acceleration=false)
   &ScaleColor(
    Scale Alpha=&FloatFromCurve(FloatCurve=DataInterface() {propertyValues {Curve {
        defaultValue=1
        keys [
         {arriveTangent=9.0725, interpMode=RCIM_Cubic, leaveTangent=9.0725, tangentMode=RCTM_User, time=0, value=0}
         {arriveTangent=-0.7567, interpMode=RCIM_Cubic, leaveTangent=-0.7567, tangentMode=RCTM_User, time=0.1408, value=0.9847}
         {arriveTangent=-0.0405, interpMode=RCIM_Cubic, leaveTangent=-0.0405, tangentMode=RCTM_User, time=0.8, value=0}
        ]
       }}})
    ScaleA=true
    ScaleRGB=false
   )
   &DynamicMaterialParameters(
    Index 0 Param 2=&MaskFloat(
     Mask Source=1
     Mask Parameter=$Particles.MeshIndex
     Mask=0
     Masked Float=&FloatFromCurve(FloatCurve=DataInterface() {propertyValues {Curve {
         defaultValue=1
         keys [
          <at 0 0 RCIM_Cubic>
          {arriveTangent=1.0101, leaveTangent=0.2896, tangentMode=RCTM_User, time=0.155, value=1}
          {arriveTangent=-0.4287, interpMode=RCIM_Cubic, leaveTangent=-0.4287, time=0.5529, value=2}
          {arriveTangent=-4.4736, interpMode=RCIM_Cubic, leaveTangent=-4.4736, time=0.6196, value=0.8008}
          <at 1 0>
         ]
        }}})
     Passthrough Float=&MaskFloat(
      Mask Source=1
      Mask Parameter=$Particles.MeshIndex
      Mask=2
      Masked Float=&FloatFromCurve(FloatCurve=DataInterface() {propertyValues {Curve {
          defaultValue=1
          keys [
           {arriveTangent=15.7454, interpMode=RCIM_Cubic, time=0, value=0}
           {arriveTangent=0.777, interpMode=RCIM_Cubic, leaveTangent=0.777, tangentMode=RCTM_User, time=0.05, value=0.63}
           {arriveTangent=1.4867, interpMode=RCIM_Cubic, leaveTangent=1.4867, tangentMode=RCTM_User, time=0.12, value=0.74}
           {arriveTangent=1.3235, interpMode=RCIM_Cubic, leaveTangent=1.3235, time=0.1452, value=0.7752}
           {arriveTangent=2.6995, interpMode=RCIM_Cubic, leaveTangent=2.6995, time=0.188, value=0.83}
           {arriveTangent=3.6975, interpMode=RCIM_Cubic, leaveTangent=3.6975, tangentMode=RCTM_User, time=0.21, value=0.95}
           {arriveTangent=0.3498, tangentMode=RCTM_User, time=0.26, value=2}
           {arriveTangent=-6.6308, interpMode=RCIM_Cubic, leaveTangent=-6.6308, tangentMode=RCTM_User, time=0.5814, value=1}
           <at 1 0>
          ]
         }}})
      Passthrough Float=&MaskFloat(
       Mask Source=1
       Mask Parameter=$Particles.MeshIndex
       Mask=4
       Masked Float=&FloatFromCurve(FloatCurve=DataInterface() {propertyValues {Curve {
           defaultValue=1
           keys [
            {arriveTangent=1.3832, interpMode=RCIM_Cubic, leaveTangent=1.3832, tangentMode=RCTM_User, time=0.1446, value=0}
            {arriveTangent=5.9809, interpMode=RCIM_Cubic, tangentMode=RCTM_Break, time=0.2876, value=0.65}
            {arriveTangent=3.149, tangentMode=RCTM_User, time=0.3596, value=0.8}
            {tangentMode=RCTM_User, time=0.3846, value=1}
            {arriveTangent=-1.5975, interpMode=RCIM_Cubic, leaveTangent=-1.5975, time=0.5446, value=2}
            {arriveTangent=-4.3922, interpMode=RCIM_Cubic, leaveTangent=-4.3922, time=0.6481, value=0.5792}
            <at 1 0>
           ]
          }}})
       Passthrough Float=&MaskFloat(
        Mask Source=1
        Mask Parameter=$Particles.MeshIndex
        Mask=6
        Masked Float=&FloatFromCurve(FloatCurve=DataInterface() {propertyValues {Curve {
            defaultValue=1
            keys [
             {arriveTangent=1.5932, interpMode=RCIM_Cubic, leaveTangent=1.5932, tangentMode=RCTM_User, time=0.4454, value=0}
             {arriveTangent=1.4091, leaveTangent=-0.3339, tangentMode=RCTM_User, time=0.5036, value=1}
             {tangentMode=RCTM_User, time=0.5718, value=1.9762}
             {arriveTangent=-2.1186, interpMode=RCIM_Cubic, leaveTangent=-1.1582, tangentMode=RCTM_Break, time=0.6461, value=0.1614}
             <at 0.6924 0>
            ]
           }}})
        Passthrough Float=0
       )
      )
     )
    )
    Index 0 Param 3=1
    Index 1 Param 1=$System.StarBrightness
    Module.Index0Float0Write=false
    Module.Index0Float1Write=true
    Module.Index0Float2Write=true
    Module.Index0Float3Write=false
    Module.Index1Float0Write=true
    Module.Index1Float1Write=false
    Module.Index1Float2Write=false
    Module.Index1Float3Write=false
    Write Parameter Index 1=::NewEnumerator1
   )
   /SetVariables(
    Particles.Color=&MaskLinearColor(
     Mask Source=1
     Mask Parameter=$Particles.MeshIndex
     Mask=1
     Masked Linear Color=&Multiply_LinearColorByFloat(
      Linear Color=$System.StarColor
      Float=&Multiply_Float(
       A=$System.OutlineIntensity
       B=&FloatFromCurve(FloatCurve=DataInterface() {propertyValues {Curve {
           defaultValue=0
           keys [
            {interpMode=RCIM_Cubic, tangentMode=RCTM_User, time=0.0589, value=0}
            <at 0.291 0.8941>
            {arriveTangent=-1.7656, interpMode=RCIM_Cubic, leaveTangent=-1.7656, time=0.5035, value=1}
            {arriveTangent=-2.7286, interpMode=RCIM_Cubic, leaveTangent=-2.7286, time=0.5881, value=0.3696}
            {arriveTangent=-0.0243, tangentMode=RCTM_User, time=0.87, value=0}
           ]
          }}})
      )
     )
     Passthrough Linear Color=&MaskLinearColor(
      Mask Source=1
      Mask Parameter=$Particles.MeshIndex
      Mask=3
      Masked Linear Color=&Multiply_LinearColorByFloat(
       Linear Color=$System.StarColor
       Float=&Multiply_Float(
        A=$System.OutlineIntensity
        B=&FloatFromCurve(FloatCurve=DataInterface() {propertyValues {Curve {
            defaultValue=1
            keys [
             {arriveTangent=-0.1316, time=0, value=0}
             {arriveTangent=3.0253, interpMode=RCIM_Cubic, leaveTangent=3.0253, time=0.1632, value=0}
             {arriveTangent=6.2038, leaveTangent=-0.0477, tangentMode=RCTM_User, time=0.3105, value=0.9392}
             {arriveTangent=-1.4109, interpMode=RCIM_Cubic, leaveTangent=-1.4109, time=0.5141, value=1.0235}
             {arriveTangent=-6.6134, interpMode=RCIM_Cubic, leaveTangent=-6.6134, tangentMode=RCTM_User, time=0.578, value=0.5617}
             {arriveTangent=-1.3105, interpMode=RCIM_Cubic, leaveTangent=-1.3105, tangentMode=RCTM_User, time=0.7028, value=0}
            ]
           }}})
       )
      )
      Passthrough Linear Color=&MaskLinearColor(
       Mask Source=1
       Mask Parameter=$Particles.MeshIndex
       Mask=5
       Masked Linear Color=&Multiply_LinearColorByFloat(
        Linear Color=$System.StarColor
        Float=&Multiply_Float(
         A=$System.OutlineIntensity
         B=&FloatFromCurve(FloatCurve=DataInterface() {propertyValues {Curve {
             defaultValue=0
             keys [
              <at 0 0>
              {arriveTangent=1.6831, interpMode=RCIM_Cubic, leaveTangent=1.6831, time=0.3325, value=0}
              <at 0.5316 0.8948>
              {arriveTangent=-3.8948, interpMode=RCIM_Cubic, leaveTangent=-3.8948, time=0.5732, value=0.9882}
              {arriveTangent=-8.2948, interpMode=RCIM_Cubic, leaveTangent=-8.2948, tangentMode=RCTM_User, time=0.6263, value=0.5261}
              {tangentMode=RCTM_User, time=0.7214, value=0}
             ]
            }}})
        )
       )
       Passthrough Linear Color=&MaskLinearColor(
        Mask Source=1
        Mask Parameter=$Particles.MeshIndex
        Mask=7
        Masked Linear Color=&Multiply_LinearColorByFloat(
         Linear Color=$System.StarColor
         Float=&Multiply_Float(
          A=$System.OutlineIntensity
          B=&FloatFromCurve(FloatCurve=DataInterface() {propertyValues {Curve {
              defaultValue=0
              keys [
               {arriveTangent=5.487, interpMode=RCIM_Cubic, leaveTangent=5.487, tangentMode=RCTM_User, time=0.4, value=0}
               {arriveTangent=-0.3774, tangentMode=RCTM_User, time=0.6468, value=0.9847}
               {arriveTangent=-4.3337, interpMode=RCIM_Cubic, leaveTangent=-4.3337, time=0.7078, value=1}
               {arriveTangent=-5.0443, interpMode=RCIM_Cubic, leaveTangent=-5.0443, time=0.7857, value=0.3826}
               {arriveTangent=-0.0405, interpMode=RCIM_Cubic, leaveTangent=-0.0405, tangentMode=RCTM_User, time=0.9061, value=0}
              ]
             }}})
         )
        )
        Passthrough Linear Color=$Particles.Color
       )
      )
     )
    )
    Particles.CustomSorting:NiagaraFloat=&SelectFloatFromArray(Direct Array Index=$Particles.MeshIndex, Float Selection Array=DataInterface() {propertyValues {FloatData [0,-1,0,-1,0,-1,0,-1]}})
    Particles.DynamicMaterialParameter:Vector4f=/Niagara/DynamicInputs/Helpers/MaskVector4(
     Mask Source:NiagaraInt32=1
     Mask Parameter:NiagaraInt32=$Particles.MeshIndex
     Mask:NiagaraInt32=&SelectIntFromArray(Direct Array Index=$Particles.MeshIndex, Int Selection Array=DataInterface() {propertyValues {IntData [1,3,5]}})
     Masked Vector4:Vector4f=&MakeVector4(X=2, Y=0, Z=0, W=0)
     Passthrough Vector4:Vector4f=&MaskVector4(
      Mask Source:NiagaraInt32=1
      Mask Parameter:NiagaraInt32=$Particles.MeshIndex
      Mask:NiagaraInt32=7
      Masked Vector4:Vector4f=&MakeVector4(X=3, Y=0, Z=0, W=0)
      Passthrough Vector4:Vector4f=$Particles.DynamicMaterialParameter
     )
    )
    Particles.DynamicMaterialParameter1:Vector4f=&MaskVector4(
     Mask Source:NiagaraInt32=1
     Mask Parameter:NiagaraInt32=$Particles.MeshIndex
     Mask:NiagaraInt32=&SelectIntFromArray(Direct Array Index=$Particles.MeshIndex, Int Selection Array=DataInterface() {propertyValues {IntData [1,3,5,7]}})
     Masked Vector4:Vector4f=&MakeVector4(X=1, Y=0, Z=0, W=0)
     Passthrough Vector4:Vector4f=$Particles.DynamicMaterialParameter1
    )
   ) {isSetParameters=true}
  ]
  rendererClasses [&NiagaraMeshRendererProperties]
  renderers [{rendererClass=&NiagaraMeshRendererProperties, rendererIndex=0}]
 }
 Star_Upper_Lines {
  bEnabled=false
  emitterUpdateScript [
   &EmitterState()
   &SpawnBurst_Instantaneous(Spawn Count=1, Use Spawn Probability=false)
  ]
  particleSpawnScript [&InitializeParticle(
    Color=$System.StarColor
    Lifetime=3.5
    Mesh Scale Mode=::NewEnumerator3
    Mesh Uniform Scale=$System.StarScale
    UsePositionOffset=false
    Write Material Random=true
   ) {module=InitializeParticle001}]
  particleUpdateScript [
   &ParticleState()
   &ScaleMeshSize(Scale Factor=&VectorFromFloat(Value=&FloatFromCurve(FloatCurve=DataInterface() {propertyValues {Curve {
         defaultValue=1
         keys [
          {tangentMode=RCTM_User, time=0, value=1}
          {arriveTangent=-0.1707, interpMode=RCIM_Cubic, leaveTangent=-0.1707, tangentMode=RCTM_User, time=0.4969, value=1}
          {arriveTangent=1.4009, interpMode=RCIM_Cubic, leaveTangent=1.4009, tangentMode=RCTM_User, time=1, value=1.3}
         ]
        }}})))
   &DynamicMaterialParameters(
    Index 0 Param 2=&FloatFromCurve(FloatCurve=DataInterface() {propertyValues {Curve {
        defaultValue=1
        keys [
         {arriveTangent=15.7454, interpMode=RCIM_Cubic, time=0, value=0}
         {arriveTangent=0.777, interpMode=RCIM_Cubic, leaveTangent=0.777, tangentMode=RCTM_User, time=0.05, value=0.63}
         {arriveTangent=1.4867, interpMode=RCIM_Cubic, leaveTangent=1.4867, tangentMode=RCTM_User, time=0.12, value=0.74}
         {arriveTangent=1.3235, interpMode=RCIM_Cubic, leaveTangent=1.3235, time=0.1452, value=0.7752}
         {arriveTangent=2.6995, interpMode=RCIM_Cubic, leaveTangent=2.6995, time=0.188, value=0.83}
         {arriveTangent=3.6975, interpMode=RCIM_Cubic, leaveTangent=3.6975, tangentMode=RCTM_User, time=0.21, value=0.95}
         {arriveTangent=0.3498, tangentMode=RCTM_User, time=0.26, value=2}
         {arriveTangent=-2.7027, interpMode=RCIM_Cubic, leaveTangent=-2.7027, time=0.6, value=1}
         <at 1 0>
        ]
       }}})
    Index 0 Param 3=1
    Index 1 Param 1=$System.StarBrightness
    Module.Index0Float0Write=false
    Module.Index0Float1Write=true
    Module.Index0Float2Write=true
    Module.Index0Float3Write=false
    Module.Index1Float0Write=true
    Module.Index1Float1Write=false
    Module.Index1Float2Write=false
    Module.Index1Float3Write=false
    Write Parameter Index 1=::NewEnumerator1
   )
   &SolveForcesAndVelocity(Clamp Velocity=false, Limit Acceleration=false)
   &ScaleColor(
    Scale Alpha=&FloatFromCurve(FloatCurve=DataInterface() {propertyValues {Curve {
        defaultValue=1
        keys [
         {arriveTangent=9.0725, interpMode=RCIM_Cubic, leaveTangent=9.0725, tangentMode=RCTM_User, time=0, value=0}
         {arriveTangent=-1.092, interpMode=RCIM_Cubic, leaveTangent=-1.092, tangentMode=RCTM_User, time=0.1408, value=0.9847}
         {arriveTangent=-0.0405, interpMode=RCIM_Cubic, leaveTangent=-0.0405, tangentMode=RCTM_User, time=0.9, value=0}
        ]
       }}})
    ScaleA=true
    ScaleRGB=false
   )
   /SetVariables(
    Particles.DMP1:Vector4f=&MakeVector4(X=2, Y=0, Z=0, W=0)
    Particles.OutlineColor:LinearColor=&Multiply_LinearColorByFloat(
     Linear Color=$System.StarColor
     Float=&Multiply_Float(
      A=$System.OutlineIntensity
      B=&FloatFromCurve(FloatCurve=DataInterface() {propertyValues {Curve {
          defaultValue=1
          keys [
           {arriveTangent=-0.1316, time=0, value=0}
           {arriveTangent=3.3638, interpMode=RCIM_Cubic, leaveTangent=3.3638, time=0.15, value=0}
           {arriveTangent=6.2038, leaveTangent=-0.0477, tangentMode=RCTM_User, time=0.2973, value=1}
           {arriveTangent=-1.7894, interpMode=RCIM_Cubic, leaveTangent=-1.7894, time=0.501, value=1}
           {arriveTangent=-3.3441, interpMode=RCIM_Cubic, leaveTangent=-3.3441, time=0.5553, value=0.5382}
           {arriveTangent=-1.3105, interpMode=RCIM_Cubic, leaveTangent=-1.3105, tangentMode=RCTM_User, time=0.8, value=0}
          ]
         }}})
     )
    )
   ) {isSetParameters=true}
  ]
  rendererClasses [&NiagaraMeshRendererProperties]
  renderers [
   {rendererClass=&NiagaraMeshRendererProperties, rendererIndex=0}
   {rendererClass=&NiagaraMeshRendererProperties, rendererIndex=1}
  ]
 }
}
summary {
 systemName=NS_Backplates_Star_Trace
 emitters [
  {bEnabled=false, emitterName=Star_Body, rendererClasses [&NiagaraMeshRendererProperties]}
  {bEnabled=false, emitterName=Star_Upper_Lines, rendererClasses [&NiagaraMeshRendererProperties]}
  {bEnabled=false, emitterName=Star_Lower_Lines, rendererClasses [&NiagaraMeshRendererProperties]}
  {emitterName=Star_Point, rendererClasses [&NiagaraMeshRendererProperties]}
  {emitterName=Boost, rendererClasses [&NiagaraSpriteRendererProperties]}
  {emitterName=Little_Stars, rendererClasses [&NiagaraSpriteRendererProperties]}
  {emitterName=Light, rendererClasses [&NiagaraLightRendererProperties]}
 ]
 userVariables []
}
